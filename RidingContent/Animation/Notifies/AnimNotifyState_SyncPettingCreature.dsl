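; Anim notify state that keeps a rider and their active ridable creature in sync during a petting animation.
; On begin it reads the petting data from the rider's active RidableComponent and calls PlaySyncedMontage on the rider and creature anim instances.
; GetRiderPettingMontage picks the rider's male or female montage from the pawn's character gender.
; On end it stops only the creature's montage, with zero blend-out.
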(fn GetRiderPettingMontage
  (params (RiderActor :type Actor%20Object%20Reference))
  (bind _asfort_player_pawn (Utilities|Casting|CastToFortPlayerPawn RiderActor)
    (:then
      (Utilities|IsValid (Actor|GetComponentbyClass _asfort_player_pawn "/Script/RidingCodeRuntime.RiderComponent")
        (:Is%20Valid
          (Utilities|IsValid (Variables|/Script/RidingCodeRuntime.RiderComponent|GetDoNotUse_ActiveRidable (Actor|GetComponentbyClass _asfort_player_pawn "/Script/RidingCodeRuntime.RiderComponent"))
            (:Is%20Valid
              (switch Utilities|FlowControl|Switch|SwitchonEFortCustomGender (Customization|GetCharacterGender _asfort_player_pawn)
                (:Invalid
                  _)
                (:Male
                  _)
                (:Female
                  (bind (_returnvalue_bispettablecreature _returnvalue_ridermalepettingmontageinfo _returnvalue_riderfemalepettingmontageinfo) (Utilities|Struct|BreakRidingPettingData (Class|RidableComponent|GetPettingData (Variables|/Script/RidingCodeRuntime.RiderComponent|GetDoNotUse_ActiveRidable (Actor|GetComponentbyClass _asfort_player_pawn "/Script/RidingCodeRuntime.RiderComponent")))))
                  (return (Utilities|Struct|BreakFortGameplayAbilityMontageInfo _returnvalue_riderfemalepettingmontageinfo)))
                (:Both
                  _))
              (bind (_returnvalue_bispettablecreature_2 _returnvalue_ridermalepettingmontageinfo_2) (Utilities|Struct|BreakRidingPettingData (Class|RidableComponent|GetPettingData (Variables|/Script/RidingCodeRuntime.RiderComponent|GetDoNotUse_ActiveRidable (Actor|GetComponentbyClass _asfort_player_pawn "/Script/RidingCodeRuntime.RiderComponent")))))
              (return (Utilities|Struct|BreakFortGameplayAbilityMontageInfo _returnvalue_ridermalepettingmontageinfo_2)))
            (:Is%20Not%20Valid
              _)))
        (:Is%20Not%20Valid
          _))
      (return 0))
    (:CastFailed
      _)))

(fn Received_NotifyBegin
  (params (MeshComp :type Skeletal%20Mesh%20Component%20Object%20Reference) (Animation :type Anim%20Sequence%20Base%20Object%20Reference) (TotalDuration :type Float%20%28single-precision%29) (EventReference :type Anim%20Notify%20Event%20Reference%20Structure%20%28by%20ref%29))
  (bind (_returnvalue_bispettablecreature _returnvalue_ridermalepettingmontageinfo _returnvalue_riderfemalepettingmontageinfo _returnvalue_ridablepettingmontageinfo) (Utilities|Struct|BreakRidingPettingData (Class|RidableComponent|GetPettingData (Variables|/Script/RidingCodeRuntime.RiderComponent|GetDoNotUse_ActiveRidable (Actor|GetComponentbyClass (Components|GetOwner MeshComp) "/Script/RidingCodeRuntime.RiderComponent")))))
  (Animation|Sync|PlaySyncedMontage (Components|SkeletalMesh|GetAnimInstance MeshComp) (CallFunction|GetRiderPettingMontage (Components|GetOwner MeshComp)) (Components|SkeletalMesh|GetAnimInstance (Actor|GetComponentbyClass (Components|GetOwner (Variables|/Script/RidingCodeRuntime.RiderComponent|GetDoNotUse_ActiveRidable (Actor|GetComponentbyClass (Components|GetOwner MeshComp) "/Script/RidingCodeRuntime.RiderComponent"))) "/Script/Engine.SkeletalMeshComponent")) (Utilities|Struct|BreakFortGameplayAbilityMontageInfo _returnvalue_ridablepettingmontageinfo))
  (return false))

(fn Received_NotifyEnd
  (params (MeshComp :type Skeletal%20Mesh%20Component%20Object%20Reference) (Animation :type Anim%20Sequence%20Base%20Object%20Reference) (EventReference :type Anim%20Notify%20Event%20Reference%20Structure%20%28by%20ref%29))
  (bind (_returnvalue_bispettablecreature _returnvalue_ridermalepettingmontageinfo _returnvalue_riderfemalepettingmontageinfo _returnvalue_ridablepettingmontageinfo) (Utilities|Struct|BreakRidingPettingData (Class|RidableComponent|GetPettingData (Variables|/Script/RidingCodeRuntime.RiderComponent|GetDoNotUse_ActiveRidable (Actor|GetComponentbyClass (Components|GetOwner MeshComp) "/Script/RidingCodeRuntime.RiderComponent")))))
  (Animation|Montage|MontageStop (Components|SkeletalMesh|GetAnimInstance (Actor|GetComponentbyClass (Components|GetOwner (Variables|/Script/RidingCodeRuntime.RiderComponent|GetDoNotUse_ActiveRidable (Actor|GetComponentbyClass (Components|GetOwner MeshComp) "/Script/RidingCodeRuntime.RiderComponent"))) "/Script/Engine.SkeletalMeshComponent")) 0.0 (Utilities|Struct|BreakFortGameplayAbilityMontageInfo _returnvalue_ridablepettingmontageinfo))
  (return false))
